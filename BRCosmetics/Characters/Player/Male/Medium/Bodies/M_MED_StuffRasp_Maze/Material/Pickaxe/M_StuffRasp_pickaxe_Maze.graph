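# Two-sided skeletal-mesh material for the Maze variant of the StuffRasp pickaxe, with a smoky emissive glow over a near-black base.
# The MF_CharacterParent_2023 PreFX and PostFX functions wrap attribute overrides for emissive, base color, metallic, specular and SRM-driven roughness.
# Emissive math combines Fresnel terms, panning world-aligned noise sampled in pre-skinned local space, and masked detail smoke that is zeroed on lower quality tiers.
# A reactive color switch can tint the glow.
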Material /BRCosmetics/Characters/Player/Male/Medium/Bodies/M_MED_StuffRasp_Maze/Material/Pickaxe/M_StuffRasp_pickaxe_Maze
graph {
 properties {
  allowTranslucentCustomDepthWrites=true
  bAutomaticallySetUsageInEditor=false
  bCastRayTracedShadows=false
  bUseMaterialAttributes=true
  bUsedWithMeshDeformer=true
  bUsedWithSkeletalMesh=true
  ditherOpacityMask=true
  translucencyLightingMode=TLM_SurfacePerPixelLighting
  twoSided=true
  usedShadingModels=Default Lit
 }
 nodes [
  n1 = SetMaterialAttributes!(
   MaterialAttributes=NamedRerouteUsage()
   Emissive Color=NamedRerouteUsage()
   Base Color=ScalarParameter() {defaultValue=0.01, parameterName=StuffRasp_BaseColor}
   Metallic=ScalarParameter() {parameterName=StuffRasp_Metallic}
   Specular=ScalarParameter() {defaultValue=0.1, parameterName=StuffRasp_Specular}
   Roughness=NamedRerouteUsage()
   Anisotropy
   Opacity
   Opacity Mask
  ) {attributeSetTypes [
    #1
    #2
    #3
    #4
    #5
    #6
    #7
    #8
   ]}
  n2 = MaterialFunctionCall() {materialFunction=&MF_CharacterParent_2023_PreFX}
  MaterialFunctionCall(Material=@n1, M=@n2.M, Pre Skinned Local Position=@n2.M, Distance Blend=@n2.M, Fresnel=@n2.M) {materialFunction=&MF_CharacterParent_2023_PostFX}
  n3 = Multiply(A=NamedRerouteUsage())
  NamedRerouteDeclaration(None=TextureSample(UVs=Panner(Coordinate=@n3) {speedY=0.1}) {samplerType=SAMPLERTYPE_Grayscale, texture=/Game/Global/Textures/Default/Noise/LowResBlurredNoiseNormalized}.R) {name=GlobalNoise, nodeColor=<rgba 0.4531 0 1 1>}
  n4 = ScalarParameter() {group=TriplanarUV, parameterName=TriplanarUV_3D_UseClamp}
  n5 = ScalarParameter() {group=TriplanarUV, parameterName=TriplanarUV_3D_FlipUV}
  n6 = VectorParameter() {group=TriplanarUV, parameterName=TriplanarUV_3D_CenterPosition, defaultValue=<rgba 0 0 0 1>}
  n7 = ScalarParameter() {defaultValue=30, group=TriplanarUV, parameterName=TriplanarUV_3D_UVSize}
  NamedRerouteDeclaration(None=OneMinus(None=MaterialFunctionCall!(TriplanarUV_UseClamp?=@n4, TriplanarUV_FlipUV=@n5, TriplanarUV_CenterPosition=@n6.RGB, TriplanarUV_UVScale=@n7) {materialFunction=/BRCosmetics/Characters/Player/Male/Medium/Bodies/M_MED_BirchPalace/Skins/Gas/Materials/MF_TriplanarUV}.UV_FrontProjection)) {name=TriplannarUV_front_1, nodeColor=<rgba 1 0.3516 0 1>}
  NamedRerouteDeclaration(None=Constant()) {name=Dot_Mask_1, nodeColor=<rgba 0 0.4609 1 1>}
  n8 = TextureSample() {samplerType=SAMPLERTYPE_Masks, texture=/BRCosmetics/Weapons/FORT_Melee/Pickaxe_StuffRasp/Textures/T_Pickaxe_StuffRasp_TA_Mask2}
  NamedRerouteDeclaration(None=@n8.R) {name=AO, nodeColor=<rgba 0 1 0.5781 1>}
  NamedRerouteDeclaration(None=@n8.G) {name=CRV, nodeColor=<rgba 1 0.2578 0 1>}
  n9 = NamedRerouteUsage()
  n10 = Fresnel(ExponentIn=ScalarParameter() {defaultValue=0.8156, group=StuffRasp_axe, parameterName=VertexNormal_FresnelPower}, Normal=VertexNormalWS())
  n11 = ScalarParameter() {defaultValue=-2, group=StuffRasp_axe, parameterName=PixNormal_FresnelPowerMin}
  n12 = ScalarParameter() {defaultValue=6.5, group=StuffRasp_axe, parameterName=PixNormal_FresnelPowerMax}
  n13 = TextureSampleParameter2D() {
   group=00_Texture Inputs
   parameterName=Mask
   samplerSource=SSM_Wrap_WorldGroupSettings
   samplerType=SAMPLERTYPE_Masks
   sortPriority=3
   texture=&T_BlackMask
  }
  n14 = Saturate(None=Fresnel(ExponentIn=LinearInterpolate(A=@n11, B=@n12, Alpha=@n13.G), Normal=PixelNormalWS()))
  NamedRerouteDeclaration(None=Add(A=Multiply(A=@n9, B=@n10), B=Power(Base=@n14) {constExponent=1.5})) {name=Main_Fresnnel, nodeColor=<rgba 0 0.0859 1 1>}
  n15 = TextureObjectParameter() {
   group=StuffRasp_axe
   parameterName=NoiseTexture
   samplerType=SAMPLERTYPE_LinearColor
   texture=/Game/Weapons/WeaponSkins/Textures/T_Swirly_Noise_highContrast1
  }
  n16 = VectorParameter() {group=StuffRasp_axe, parameterName=NoiseWorldTileScale, defaultValue=<rgba 111 111 100 0>}
  n17 = Subtract(
   A=VertexInterpolator(VS=LocalPosition() {localOrigin=InstancePreSkinning}.XYZ).PS
   B=Multiply(A=Time(), B=VectorParameter() {group=StuffRasp_axe, parameterName=NoisePanSpeed, defaultValue=<rgba -0.66 -1 -2 0>}.RGB)
  )
  n18 = VertexInterpolator(VS=PreSkinnedNormal())
  NamedRerouteDeclaration(None=ComponentMask(None=MaterialFunctionCall(TextureObject=@n15, TextureSize=@n16.RGB, WorldPosition=@n17, World Space Normal=@n18.PS) {materialFunction=/Game/Weapons/WeaponSkins/Wraps/MaterialFunctions/WorldAlignedTexture}.XYZ Texture) {r=true}) {name=SmokeNoise, nodeColor=<rgba 1 0 0.3281 1>}
  n19 = Constant()
  n20 = Multiply(
   A=MaterialFunctionCall!(In_FireSmoke_TAMask, In_FireSmoke_UV=NamedRerouteUsage()) {materialFunction=/BRCosmetics/Gadgets/Assets/VinderTech_GliderChute/Glider_RampCanal/Materials/MFW_FlowingSmoke}.Result
   B=ScalarParameter() {defaultValue=0.5, group=StuffRasp_axe, parameterName=DetailSmoke_Intensity}
  ) {constB=0.5}
  n21 = MaterialFunctionCall!(
   Low _ PC Low and Med=@n19
   "High _ XB1, PS4"=@n20
   "Epic _ PC High, XSX, PS5"=@n20
   --------
   "Mobile Low _ PC Perf, Android"=@n19
   Mobile Med _ Switch=@n19
  ) {materialFunction=/Game/Global/MaterialFunctions/TechArt/MF_Fortnite_QualitySwitch_Scalar}
  n22 = TextureSample(UVs=TextureCoordinate()) {
   samplerSource=SSM_Wrap_WorldGroupSettings
   samplerType=SAMPLERTYPE_Masks
   texture=&T_Pickaxe_StuffRasp_TA_Mask2
  }
  n23 = ScalarParameter() {defaultValue=0.2665, group=StuffRasp_axe, parameterName=DetailSmokeBrightness}
  n24 = ComponentMask(None=Power(Base=Multiply(A=@n22.R, B=@n23), Exp=ScalarParameter() {defaultValue=0.4812, group=StuffRasp_axe, parameterName=DetailSmokeBrightnessPower})) {r=true}
  n25 = NamedRerouteUsage()
  n26 = LinearInterpolate(A=Constant() {r=1}, B=Constant() {r=-1}, Alpha=Power(Base=ComponentMask(None=@n25) {g=true}) {constExponent=3})
  NamedRerouteDeclaration(None=Multiply(A=Multiply(A=@n21.Result, B=@n24), B=Saturate(None=@n26)) {constB=0.5}) {name=DetailSmoke, nodeColor=<rgba 1 0 0.8906 1>}
  n27 = NamedRerouteUsage()
  n28 = Fresnel(ExponentIn=ScalarParameter() {defaultValue=2.1, group=StuffRasp_axe, parameterName=FresnelGlobal}, Normal=VertexNormalWS())
  n29 = Constant() {r=0.8}
  n30 = NamedRerouteUsage()
  n31 = LinearInterpolate(A=Constant() {r=1.3}, B=Constant() {r=1}, Alpha=NamedRerouteUsage())
  n32 = Multiply(
   A=LinearInterpolate(A=Constant() {r=0.1}, Alpha=NamedRerouteUsage()) {constB=0.8}
   B=Power(Base=Multiply(A=@n30, B=@n31), Exp=Constant() {r=6})
  )
  n33 = NamedRerouteUsage()
  n34 = Constant() {r=-0.1422}
  n35 = PixelNormalWS()
  n36 = LinearInterpolate(
   A=ScalarParameter() {defaultValue=0.3, group=StuffRasp_axe, parameterName=NoiseScaleOuter}
   B=ScalarParameter() {defaultValue=0.75, group=StuffRasp_axe, parameterName=NoiseScaleInner}
   Alpha=Saturate(None=Fresnel(ExponentIn=@n34, Normal=@n35))
  )
  n37 = Multiply(A=Saturate(None=Multiply(A=@n33, B=@n36)), B=Multiply(A=NamedRerouteUsage(), B=Constant() {r=2}))
  n38 = Constant()
  n39 = NamedRerouteUsage()
  n40 = Multiply(A=Multiply(A=NamedRerouteUsage(), B=NamedRerouteUsage()), B=Saturate(None=Multiply(A=@n39) {constB=2}))
  n41 = Constant() {r=2}
  n42 = Multiply(
   A=Add(A=Multiply(A=@n28, B=@n29), B=Max(A=@n32, B=@n37))
   B=LinearInterpolate(A=LinearInterpolate(B=@n38, Alpha=@n40) {constA=1, constB=0}, B=LinearInterpolate(B=@n41, Alpha=@n40) {constB=0}, Alpha=NamedRerouteUsage())
  )
  NamedRerouteDeclaration(None=Power(Base=Add(A=@n27, B=@n42), Exp=ScalarParameter() {defaultValue=1.6, group=StuffRasp_axe, parameterName=GlobalContrast})) {name=OutEmissvie, nodeColor=<rgba 1 0 0.9844 1>}
  n43 = TextureSampleParameter2D() {
   desc="R = Spec , G = Roughness, B= Metal"
   group=00_Texture Inputs
   parameterName=SRM
   samplerSource=SSM_Wrap_WorldGroupSettings
   samplerType=SAMPLERTYPE_Masks
   sortPriority=1
   texture=&T_Fortnite_Default_S
  }
  NamedRerouteDeclaration(None=LinearInterpolate(
    A=ScalarParameter() {defaultValue=0.5, parameterName=StuffRasp_RoughnessMin}
    B=ScalarParameter() {defaultValue=1, parameterName=StuffRasp_RoughnessMax}
    Alpha=@n43.G
   ) {constA=0.5, constB=0.9}) {name=OutRoughness, nodeColor=<rgba 1 0.7734 0 1>}
  n44 = VectorParameter() {parameterName=StuffRasp_gloveReactiveColor, defaultValue=<rgba 1 0.05 0.1502 1>}
  n45 = ScalarParameter() {parameterName=StuffRasp_gloveColorSwitch}
  n46 = ScalarParameter() {defaultValue=1, parameterName=StuffRasp_gloveReactive}
  n47 = Multiply(A=@n45, B=@n46)
  n48 = Power(Base=NamedRerouteUsage(), Exp=LinearInterpolate(Alpha=@n45) {constA=1, constB=1.3})
  NamedRerouteDeclaration(None=Multiply(A=LinearInterpolate(B=@n44.RGB, Alpha=@n47) {constA=1}, B=Multiply(A=@n48, B=@n46))) {name=ShoesEmissive, nodeColor=<rgba 0.4531 0 1 1>}
  NamedRerouteDeclaration(None=GetMaterialAttributes(None=@n2.Result) {attributeGetTypes [#2]}.MaterialAttributes) {name=InMA, nodeColor=<rgba 1 0.2109 0 1>}
 ]
}
